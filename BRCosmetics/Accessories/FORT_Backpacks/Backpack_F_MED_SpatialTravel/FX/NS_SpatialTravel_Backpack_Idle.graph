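# Niagara idle effect for the SpatialTravel backpack cosmetic, emitting short-lived "derez cube" mesh and sprite particles near the pack.
# Particles spawn in an offset, stretched sphere around a skeletal mesh sample, with spawn rate falling as owner speed rises.
# Drag, curved gravity and inherited movement (stronger when skydiving) drive them, and a volume kills strays.
# Mesh scale and alpha drop to zero in vehicles or ADS; color dims at night.
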NiagaraSystem /BRCosmetics/Accessories/FORT_Backpacks/Backpack_F_MED_SpatialTravel/FX/NS_SpatialTravel_Backpack_Idle
dependencies {
 usedDataInterfaces [&NiagaraDataInterfaceCurve, &NiagaraDataInterfaceSkeletalMesh]
 usedDynamicInputs []
 usedModules [
  &SystemState
  &self:SystemSpawnScript.SystemScriptSource.SystemScriptGraph.NiagaraNodeAssignment_0.SetVariables_#1
 ]
}
emitters {Derez_Cube {
  emitterUpdateScript [
   &EmitterState()
   /SetVariables(Emitter.MovementSpawnAlpha:NiagaraFloat=&ClampFloat(Float=&Divide_Float(A=&VectorLength(Vector=$Engine.Owner.Velocity), B=500))) {isSetParameters=true}
   &SpawnRate(
    Spawn Probability=0.75
    SpawnRate=&Lerp_Float(A=4, B=2, Alpha=$Emitter.MovementSpawnAlpha)
    Use Spawn Probability=false
   )
  ]
  particleSpawnScript [
   &InitializeParticle(
    Lifetime Max=0.75
    Lifetime Min=0.5
    Lifetime Mode=::NewEnumerator1
    Mesh Scale Mode=::NewEnumerator1
    Mesh Uniform Scale Max=0.5
    Mesh Uniform Scale Min=0.375
    Sprite Size Mode=::NewEnumerator1
    Uniform Sprite Size Max=8
    Uniform Sprite Size Min=6
    UsePositionOffset=false
    Write Material Random=true
   )
   &InitialMeshOrientation(Use Rotation Vector=false)
   &SkeletalMeshLocation(
    MaskSpawnGroup=false
    Mesh Sampling Type=::NewEnumerator5
    Orientation Sampling=::NewEnumerator1
    Position Sampling=::NewEnumerator2
    Sampled Position Offset=<0 1 8>
    Skeletal Mesh=$System.SkeletalMesh
    Socket Sampling Mode=::NewEnumerator5
   )
   &ShapeLocation(
    Additional Quaternion Rotation=true
    Additional Yaw / Pitch / Roll=<0 0.5 -0.25>
    Hemisphere Distribution=<0.5 0.5>
    MaskSpawnGroup=false
    Non Uniform Scale=<0.25 1.5 0.25>
    Override Seed=false
    Rotation Mode=::NewEnumerator3
    Rotation Quaternion=$Particles.SkeletalMeshLocation.SampledRotation
    Shape Origin=$Particles.SkeletalMeshLocation.SampledPosition
    Sphere Radius=10
    Sphere Surface Distribution=1
    UseOwnerScale=true
   )
   &AddVelocity(
    Cone Axis=<-0.5 -1 0>
    Module.UseConeFalloff=false
    Module.UseVelDistribution=false
    Override Seed=false
    Velocity Mode=::NewEnumerator2
    Velocity Speed=&RandomRangeFloat(Minimum=22, Maximum=44)
   ) {module=AddVelocity003}
  ]
  particleUpdateScript [
   &ParticleState()
   &ScaleMeshSize(Scale Factor=&VectorFromFloat(Value=&MakeCustomFloatFromBool(
      Bool=&Boolean_OrOperation(Bool1=$User.bisInVehicle, Bool2=$Emitter.ADS)
      True Float=0
      False Float=&FloatFromCurve(FloatCurve=DataInterface() {propertyValues {Curve {
          defaultValue=0
          keys [
           {arriveTangent=9.0725, interpMode=RCIM_Cubic, leaveTangent=9.0725, tangentMode=RCTM_User, time=0, value=0}
           {arriveTangent=0.1035, interpMode=RCIM_Cubic, leaveTangent=0.1035, tangentMode=RCTM_User, time=0.1408, value=0.9847}
           {arriveTangent=0.0495, interpMode=RCIM_Cubic, leaveTangent=0.0495, tangentMode=RCTM_User, time=0.7, value=1}
           {arriveTangent=-3.1446, interpMode=RCIM_Cubic, leaveTangent=-3.1446, tangentMode=RCTM_User, time=1, value=0}
          ]
         }}})
     )))
   &ScaleSpriteSize(
    Scale Sprite Size Mode=::NewEnumerator0
    Uniform Scale Factor=&FloatFromCurve(FloatCurve=DataInterface() {propertyValues {Curve {keys [
         {arriveTangent=9.1187, interpMode=RCIM_Cubic, time=0, value=0}
         {arriveTangent=1.1429, interpMode=RCIM_Cubic, leaveTangent=1.1429, time=0.2487, value=1.1246}
         {arriveTangent=-1.497, interpMode=RCIM_Cubic, leaveTangent=-1.497, time=0.7, value=0.8}
         {arriveTangent=-3.0554, interpMode=RCIM_Cubic, leaveTangent=-3.0554, tangentMode=RCTM_User, time=1, value=0}
        ]}}})
   )
   &AccelerationForce(Acceleration=&Multiply_VectorByFloat(
     Vector=&RandomRangeVector(Minimum=<-3 -3 5>, Maximum=<3 3 15>, Evaluation Type=::NewEnumerator1)
     Float=1
    ))
   &Drag(Drag=&RandomRangeFloat(Minimum=7.5, Maximum=12.5), Use Linear Drag=true, Use Rotational Drag=true)
   &GravityForce(Gravity=&MakeVector(
     X=0
     Y=0
     Z=&FloatFromCurve(FloatCurve=DataInterface() {propertyValues {Curve {keys [<at 0.3 0>,<at 1 1>]}}}, Scale Curve=200)
    ))
   &DynamicMaterialParameters(
    Index 0 Param 1=&FloatFromCurve(
     FloatCurve=DataInterface() {propertyValues {Curve {
        defaultValue=1
        keys [
         {arriveTangent=-0.6868, interpMode=RCIM_Cubic, leaveTangent=-0.6868, tangentMode=RCTM_User, time=0, value=0.25}
         {interpMode=RCIM_Cubic, tangentMode=RCTM_User, time=1, value=0}
        ]
       }}}
     Scale Curve=0.5
    )
    Index 0 Param 3=&RandomRangeFloat(Minimum=0.75, Maximum=1.75)
    Module.Param0WriteEnabled=true
    Module.Param1WriteEnabled=true
    Module.Param2WriteEnabled=true
    Module.Param3WriteEnabled=false
   )
   &InheritSourceMovement(
    Applied Force Scale=&VectorFromFloat(Value=&FloatFromCurve(FloatCurve=DataInterface() {propertyValues {Curve {keys [<at 0 0>,<at 1 -0.5>]}}}))
    Applied Position Scale=&VectorFromFloat(Value=&MakeCustomFloatFromBool(
      Bool=$User.IsSkydiving
      True Float=&FloatFromCurve(FloatCurve=DataInterface() {propertyValues {Curve {keys [<at 0 0.95 RCIM_Cubic>,<at 1 0.8 RCIM_Cubic>]}}})
      False Float=&FloatFromCurve(FloatCurve=DataInterface() {propertyValues {Curve {keys [<at 0 0.95 RCIM_Cubic>, {arriveTangent=-0.2785, interpMode=RCIM_Cubic, leaveTangent=-0.2785, tangentMode=RCTM_User, time=1, value=0}]}}})
     ))
    ApplyForce=false
    ApplyPosition=true
    ApplyPositionFalloff=false
   ) {module=InheritSourceMovement001}
   &PointForce(Module.Use Falloff=false, Module.UseLinearForce=false)
   &SolveForcesAndVelocity(Clamp Velocity=false, Limit Acceleration=false)
   &ScaleColor(
    Scale Alpha=&MakeCustomFloatFromBool(Bool=&Boolean_OrOperation(Bool1=$User.bisInVehicle, Bool2=$Emitter.ADS), True Float=0, False Float=1)
    ScaleA=true
    ScaleRGB=true
   ) {module=ScaleColor001}
   &ScaleColor(
    Scale Alpha=&FloatFromCurve(FloatCurve=DataInterface() {propertyValues {Curve {
        defaultValue=0
        keys [
         {interpMode=RCIM_Cubic, tangentMode=RCTM_User, time=0, value=0}
         {arriveTangent=0.3339, interpMode=RCIM_Cubic, leaveTangent=0.3339, tangentMode=RCTM_User, time=0.1, value=0.9}
         {interpMode=RCIM_Cubic, tangentMode=RCTM_User, time=0.7, value=0.9}
         {tangentMode=RCTM_User, time=1, value=0}
        ]
       }}})
    Scale RGB=&VectorFromFloat(Value=&FNTimeOfDayScalar(Night Scale=0.6, Morning Scale=1, Evening Scale=1))
    ScaleA=true
    ScaleRGB=true
   ) {module=ScaleColor002}
   &CameraOffset(Camera Offset Amount=&FN_FrontEndFloatModifier(Gameplay=25, Front End=5))
   /SetVariables(Particles.Color_Glow:LinearColor=&MultiplyLinearColor(
     Linear Color=<rgba 0.0681 0.7611 1 0.1>
     Scale Factor=&MakeVector4_FromLinearColor(LinearColor=$Particles.Color)
    )) {isSetParameters=true}
   &KillParticlesInVolume(Invert Volume=true, Sphere Radius=20, Volume Origin:NiagaraPosition=$Particles.SkeletalMeshLocation.SampledPosition)
  ]
  rendererClasses [&NiagaraMeshRendererProperties, &NiagaraSpriteRendererProperties]
  renderers [
   {rendererClass=&NiagaraMeshRendererProperties, rendererIndex=0}
   {rendererClass=&NiagaraSpriteRendererProperties, rendererIndex=1}
  ]
 }}
summary {
 systemName=NS_SpatialTravel_Backpack_Idle
 emitters [{emitterName=Derez_Cube, rendererClasses [&NiagaraMeshRendererProperties, &NiagaraSpriteRendererProperties]}]
 userVariables [
  User.bIsFrontEnd {type=NiagaraBool}
  User.bIsFrontEndPreview {type=NiagaraBool}
  User.bisHandHidden {type=NiagaraBool}
  User.bisInVehicle {type=NiagaraBool}
  User.IsSkydiving {type=NiagaraBool}
 ]
}
