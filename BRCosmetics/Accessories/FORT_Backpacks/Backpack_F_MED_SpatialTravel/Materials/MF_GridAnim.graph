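# Material function for the SpatialTravel backpack that outputs an animated mask lighting cells of a grid texture.
# Offset, speed-scaled time is wrapped and floored into discrete steps, and a tiny-radius SphereMask matches the GridMask green channel to select the current row.
# That row mask multiplies SphereMask column selections around ColumeNumber.
# A Step against TheRowStartstoSplit drives the split from one column into two neighboring columns.
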MaterialFunction /BRCosmetics/Accessories/FORT_Backpacks/Backpack_F_MED_SpatialTravel/Materials/MF_GridAnim
graph {nodes [
  n1 = TextureSample() {samplerType=SAMPLERTYPE_Masks, texture=/BRCosmetics/Accessories/FORT_Backpacks/Backpack_F_MED_SpatialTravel/Textures/T_SpatialTravel_Anim}
  n2 = FunctionInput(Preview=@n1.RGB) {inputName=GridMask}
  n3 = ComponentMask(None=@n2) {g=true}
  n4 = Add(A=Time(), B=FunctionInput() {inputName=GridAnimTimeOffset, inputType=FunctionInput_Scalar})
  n5 = FunctionInput() {inputName=GridAnimSpeed, inputType=FunctionInput_Scalar, previewValue=<0.5 0 0 1>}
  n6 = Multiply(A=Frac(None=Multiply(A=@n4, B=@n5) {constB=0.5}), B=Multiply(B=ScalarParameter() {defaultValue=10}) {constA=0.1}) {constB=0.6}
  n7 = ScalarParameter() {defaultValue=10, parameterName=Param2}
  n8 = Divide(A=Floor(None=Multiply(A=@n6, B=@n7) {constB=10}), B=@n7) {constB=10}
  n9 = Add(A=Constant(), B=@n8) {constB=0}
  n10 = Constant()
  n11 = Subtract(A=FunctionInput(Preview=Constant() {r=2}) {inputName=ColumeNumber, inputType=FunctionInput_Scalar})
  n12 = Multiply(A=Add(A=@n11)) {constB=0.1}
  n13 = Constant()
  n14 = Multiply(A=Subtract(A=@n11)) {constB=0.1}
  n15 = Add(
   A=SphereMask(A=NamedRerouteUsage(), B=Add(A=@n10, B=@n12)) {attenuationRadius=0.01}
   B=SphereMask(A=NamedRerouteUsage(), B=Add(A=@n13, B=@n14)) {attenuationRadius=0.01}
  )
  n16 = SphereMask(A=NamedRerouteUsage(), B=Add(A=Constant(), B=Multiply(A=@n11) {constB=0.1})) {attenuationRadius=0.01}
  n17 = NamedRerouteUsage()
  FunctionOutput(None=Multiply(A=SphereMask(A=@n3, B=@n9) {attenuationRadius=0.01}, B=LinearInterpolate(A=@n15, B=@n16, Alpha=@n17)))
  NamedRerouteDeclaration(None=ComponentMask(None=@n2) {r=true}) {name=GridMask_R, nodeColor=<rgba 0 0.25 1 1>}
  n18 = FunctionInput() {inputName=TheRowStartstoSplit, inputType=FunctionInput_Scalar}
  NamedRerouteDeclaration(None=Step(Y=@n8, X=Multiply(B=@n18) {constA=0.1})) {name=Spliting, nodeColor=<rgba 0.5703 1 0 1>}
  ScalarParameter() {defaultValue=3, parameterName=TheRowStartstoSplit}
  ScalarParameter() {defaultValue=0.5, parameterName=GridAnimSpeed}
 ]}
