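# Material function for the Spatial Travel backpack cosmetic that layers an animated, cyan-blue sci-fi grid effect over incoming material attributes.
# Staggered MF_GridAnim calls are max-combined, while rotated pre-skinned local position, texture channels and crops build row, divider and transition masks.
# The grid normal map doubles as a time-offset source for emissive flicker, alongside a noise-distorted wiping band.
# Named reroutes feed these into base color, emissive, normal and opacity.
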MaterialFunction /BRCosmetics/Accessories/FORT_Backpacks/Backpack_F_MED_SpatialTravel/Materials/MF_SpatialTravelPack
graph {nodes [
  n1 = TextureSample!(UVs, Tex, Level, Apply View MipBias) {
   mipValueMode=TMVM_MipLevel
   samplerSource=SSM_Wrap_WorldGroupSettings
   samplerType=SAMPLERTYPE_Masks
   texture=/BRCosmetics/Accessories/FORT_Backpacks/Backpack_F_MED_SpatialTravel/Textures/T_SpatialTravel_Anim
  }
  n2 = NamedRerouteUsage()
  n3 = MaterialFunctionCall(R=LinearInterpolate(A=@n1.R, B=NamedRerouteUsage(), Alpha=@n2) {constB=0}, G=LinearInterpolate(A=@n1.G, Alpha=@n2) {constB=0.9}, B=@n1.B) {materialFunction=&AppendMany}
  n4 = NamedRerouteDeclaration(None=@n3.RGB) {name=AdjustedMask, nodeColor=<rgba 1 0.5391 0 1>}
  n5 = Max(
   A=MaterialFunctionCall!(
    GridAnimTimeOffset=Constant() {r=0.45}
    GridMask=@n4
    TheRowStartstoSplit=Constant() {r=4}
    GridAnimSpeed=Constant() {r=0.9}
    ColumeNumber=Constant() {r=2}
   ) {materialFunction=&/MF_GridAnim}.Result
   B=MaterialFunctionCall(
    GridAnimTimeOffset=Constant()
    GridMask=@n4
    TheRowStartstoSplit=Constant() {r=2}
    GridAnimSpeed=Constant() {r=1.425}
    ColumeNumber=Constant() {r=6}
   ) {materialFunction=&MF_GridAnim}.Result
  )
  n6 = MaterialFunctionCall(
   GridAnimTimeOffset=Constant() {r=0.85}
   GridMask=@n4
   TheRowStartstoSplit=Constant() {r=3}
   GridAnimSpeed=Constant() {r=0.75}
   ColumeNumber=Constant() {r=4}
  ) {materialFunction=&MF_GridAnim}
  n7 = Constant() {r=-0.24}
  n8 = Constant() {r=4}
  n9 = Constant() {r=0.95}
  n10 = Constant() {r=1}
  n11 = Max(
   A=Max(
    A=Max(A=@n5, B=@n6.Result)
    B=MaterialFunctionCall(GridAnimTimeOffset=@n7, GridMask=@n4, TheRowStartstoSplit=@n8, GridAnimSpeed=@n9, ColumeNumber=@n10) {materialFunction=&MF_GridAnim}.Result
   )
   B=MaterialFunctionCall(
    GridAnimTimeOffset=Constant() {r=-0.635}
    GridMask=@n4
    TheRowStartstoSplit=Constant() {r=3}
    GridAnimSpeed=Constant() {r=0.6}
    ColumeNumber=Constant() {r=7}
   ) {materialFunction=&MF_GridAnim}.Result
  )
  Add(A=@n11)
  n12 = LocalPosition() {localOrigin=InstancePreSkinning}
  n13 = MaterialFunctionCall(Float3=ShaderStageSwitch(PixelShader=VertexInterpolator(VS=@n12.XYZ).PS, VertexShader=@n12.XYZ)) {materialFunction=&BreakOutFloat3Components}
  n14 = MaterialFunctionCall(
   UVs=MaterialFunctionCall(X=@n13.R, Y=@n13.R) {materialFunction=&MakeFloat2}.Result
   Rotation Center=Constant()
   "Rotation Angle (0-1)"=Constant() {r=0.097}
  ) {materialFunction=&CustomRotator}
  n15 = ComponentMask(None=Multiply(A=@n14.Rotated Values)) {g=true}
  n16 = Saturate(None=Subtract(A=Step(Y=@n15) {constX=-25.39}, B=NamedRerouteUsage()))
  n17 = Step(Y=@n15) {constX=-26.7}
  Add(A=@n16, B=@n17)
  n18 = MaterialFunctionCall(Float3=NamedRerouteUsage()) {materialFunction=&BreakOutFloat3Components}
  n19 = Constant() {r=200}
  n20 = Multiply(A=@n18.R, B=@n19) {constB=50}
  n21 = Multiply(A=@n18.G, B=@n19) {constB=50}
  n22 = AppendVector(A=Divide(A=Ceil(None=@n20), B=@n19) {constB=50}, B=Divide(A=Ceil(None=@n21), B=@n19) {constB=50})
  Add(B=@n22) {constA=0.01}
  n23 = TextureSampleParameter2D() {
   group=STGridFX
   parameterName=AdditionalGrid
   samplerType=SAMPLERTYPE_Normal
   texture=/BRCosmetics/Accessories/FORT_Backpacks/Backpack_F_MED_SpatialTravel/Textures/T_SpatialTravel_Backpack_N
  }
  n24 = NamedRerouteUsage()
  Add(A=@n23.RGB, B=Multiply(A=@n24))
  n25 = Add(A=Divide(A=ComponentMask(None=@n14.Rotated Values) {r=true}) {constB=11}) {constB=0.1}
  n26 = Step(Y=@n25) {constX=0.1}
  n27 = Constant() {r=0.7}
  n28 = TwoSidedSign()
  n29 = Saturate(None=@n28)
  n30 = LinearInterpolate(A=Add(A=@n27) {constB=-0.1}, B=@n27, Alpha=@n29)
  n31 = Step(Y=@n25) {constX=0.2}
  n32 = Constant() {r=0.6}
  n33 = LinearInterpolate(A=Add(A=@n32) {constB=-0.1}, B=@n32, Alpha=@n28)
  n34 = Constant() {r=0.5}
  n35 = Add(A=@n34) {constB=-0.1}
  n36 = Max(
   A=Max(A=Multiply(A=@n26, B=@n30), B=Multiply(A=@n31, B=@n33))
   B=Multiply(A=Step(Y=@n25) {constX=0.3}, B=LinearInterpolate(A=@n35, B=@n34, Alpha=@n29))
  )
  n37 = Constant() {r=0.4}
  n38 = Multiply(A=Step(Y=@n25) {constX=0.4}, B=LinearInterpolate(A=Add(A=@n37) {constB=-0.1}, B=@n37, Alpha=@n29))
  n39 = Step(Y=@n25) {constX=0.5}
  n40 = Constant() {r=0.3}
  n41 = LinearInterpolate(A=Add(A=@n40) {constB=-0.1}, B=@n40, Alpha=@n29)
  n42 = Constant() {r=0.2}
  n43 = Add(A=@n42) {constB=-0.1}
  n44 = Max(
   A=Max(A=Max(A=@n36, B=@n38), B=Multiply(A=@n39, B=@n41))
   B=Multiply(A=Step(Y=@n25) {constX=0.6}, B=LinearInterpolate(A=@n43, B=@n42, Alpha=@n29))
  )
  AppendVector(A=@n44, B=Constant() {r=0.9})
  n45 = TextureSampleParameter2D() {
   group=06_Emissive
   parameterName=Emissive
   samplerSource=SSM_Wrap_WorldGroupSettings
   texture=/BRCosmetics/Accessories/FORT_Backpacks/Backpack_F_MED_SpatialTravel/Textures/T_SpatialTravel_Backpack_E
  }
  ComponentMask(None=@n45.B)
  ComponentMask(None=TextureCoordinate() {coordinateIndex=1}) {r=true}
  ComponentMask(None=@n14.Rotated Values) {r=true}
  Constant() {r=1}
  Constant() {r=0.5}
  Constant()
  Constant() {r=0.1}
  Constant() {r=1}
  Frac(None=Multiply(A=Time(), B=ScalarParameter() {defaultValue=0.5, group=STGridFX, parameterName=GridAnimSpeed}))
  n46 = BlendMaterialAttributes(A=NamedRerouteUsage(), B=NamedRerouteUsage(), Alpha=NamedRerouteUsage())
  n47 = GetMaterialAttributes(None=@n46) {attributeGetTypes [#1]}
  FunctionOutput(None=SetMaterialAttributes!(MaterialAttributes=@n46, OpacityMask=@n47.OpacityMask) {attributeSetTypes [#1]})
  n48 = ComponentMask() {b=true}
  LinearInterpolate(B=Multiply(A=Sign(None=@n48), B=Constant3Vector() {constant=<rgba 0.2 1 5 1>}), Alpha=NamedRerouteUsage())
  n49 = Step(Y=@n15) {constX=-11.2}
  LinearInterpolate(A=NamedRerouteUsage(), B=@n49)
  n50 = MaterialFunctionCall() {materialFunction=&BoundingBoxBased_0-1_UVW}
  MaterialFunctionCall(X=@n50.R, Y=@n50.R) {materialFunction=&MakeFloat2}
  MaterialFunctionCall() {materialFunction=/Engine/Functions/Engine_MaterialFunctions02/Utility/DebugOnOff}
  Max(A=@n11, B=NamedRerouteUsage())
  n51 = NamedRerouteUsage()
  n52 = NamedRerouteUsage()
  Max(A=@n51, B=@n52)
  Min(A=@n51, B=@n52)
  Multiply(A=@n3.RG)
  n53 = MaterialFunctionCall(X=@n50.R, Y=@n50.R) {materialFunction=&MakeFloat2}
  n54 = Constant() {r=0.5}
  n55 = Add(
   A=LocalPosition().XY
   B=AppendVector(
    A=ScalarParameter() {group=STGridFX, parameterName=BandLocalPositionOffsetX}
    B=ScalarParameter() {group=STGridFX, parameterName=BandLocalPositionOffsety}
   )
  )
  n56 = ScalarParameter() {defaultValue=100, group=STGridFX, parameterName=BandSpreadingLocalPositionDivider}
  n57 = LinearInterpolate(
   A=OneMinus(None=SphereMask(A=@n53.Result, B=@n54) {attenuationRadius=0.5, hardnessPercent=0})
   B=SphereMask(A=Divide(A=@n55, B=@n56) {constB=100}, B=Constant()) {attenuationRadius=1.5, hardnessPercent=0}
   Alpha=ScalarParameter() {group=STGridFX, parameterName=BandSpreadingUseLocalPosition}
  )
  n58 = Multiply(A=NamedRerouteUsage(), B=ScalarParameter() {defaultValue=0.1, group=STGridFX, parameterName=BandSpreadingNoiseAmount}) {constB=0.1}
  n59 = Frac(None=Multiply(A=Time(), B=ScalarParameter() {defaultValue=0.4, group=STGridFX, parameterName=BandSpreadingSpeed}))
  n60 = ScalarParameter() {defaultValue=1.3, group=STGridFX, parameterName=BandLocationMAX}
  n61 = Power(Base=SphereMask(
    A=Add(A=@n57, B=@n58)
    B=Multiply(A=@n59, B=@n60) {constB=1.1}
    Radius=ScalarParameter() {defaultValue=0.15, group=STGridFX, parameterName=BandWidth}
   ) {attenuationRadius=0.1, hardnessPercent=0})
  Multiply(A=@n61)
  n62 = Saturate(None=Subtract(A=NamedRerouteUsage(), B=NamedRerouteUsage()))
  Multiply(A=@n62)
  Multiply(A=@n45.RGB)
  Multiply()
  n63 = GetMaterialAttributes(None=FunctionInput() {inputType=FunctionInput_MaterialAttributes}) {attributeGetTypes [#2, #3]}
  n64 = NamedRerouteDeclaration(None=@n63.EmissiveColor) {name=EmissiveRaw, nodeColor=<rgba 0.7578 0 1 1>}
  n65 = Multiply(A=VectorParameter() {group=STGridFX, parameterName=Emissive_Dividers, defaultValue=<rgba 0.0532 0.3084 0.6094 1>}.RGB) {constB=1.5}
  n66 = SetMaterialAttributes!(
   MaterialAttributes=@n63.MaterialAttributes
   OpacityMask
   EmissiveColor=LinearInterpolate(A=@n64, B=Max(A=@n64, B=@n65), Alpha=NamedRerouteUsage())
  ) {
   desc=Removing the divider strip of the last low
   attributeSetTypes [#1, #3]
  }
  NamedRerouteDeclaration(None=@n66) {name=MatIn, nodeColor=<rgba 0 1 0.5078 1>}
  n67 = NamedRerouteUsage()
  n68 = GetMaterialAttributes(None=@n67) {attributeGetTypes [
    #4
    #3
    #1
    #5
   ]}
  n69 = VectorParameter() {parameterName=BaseColor_TransitionGrids, defaultValue=<rgba 0.0168 0.0643 0.1198 1>}
  n70 = VectorParameter() {parameterName=BaseColor_TransitionGrids_B, defaultValue=<rgba 0.0491 0.1874 0.349 1>}
  n71 = NamedRerouteUsage()
  n72 = NamedRerouteUsage()
  n73 = LinearInterpolate(
   A=@n72
   B=Multiply(A=Add(B=@n72), B=ScalarParameter() {defaultValue=3, group=STGridFX, parameterName=BandEmissiveMult})
   Alpha=NamedRerouteUsage()
  )
  n74 = LinearInterpolate(A=LinearInterpolate(Alpha=NamedRerouteUsage()) {constA=1, constB=0}, Alpha=NamedRerouteUsage())
  n75 = SetMaterialAttributes!(
   MaterialAttributes=@n67
   BaseColor=LinearInterpolate(A=@n68.BaseColor, B=LinearInterpolate(A=@n69.RGB, B=@n70.RGB, Alpha=@n71), Alpha=@n62)
   EmissiveColor=NamedRerouteDeclaration(None=Saturate(None=@n73)) {name=EmissiveFinal, nodeColor=<rgba 0 0.9062 1 1>}
   Normal=NamedRerouteUsage()
   OpacityMask=NamedRerouteDeclaration(None=@n74) {name=OpacityMask, nodeColor=<rgba 0 1 0.8359 1>}
   Roughness=@n68.Roughness
   Opacity=Multiply(A=@n68.OpacityMask, B=@n74)
  ) {attributeSetTypes [
    #4
    #3
    #2
    #1
    #5
    #6
   ]}
  NamedRerouteDeclaration(None=@n75) {name=MatOut, nodeColor=<rgba 0.2656 0 1 1>}
  n76 = LinearInterpolate(A=@n23.RGB, B=@n24, Alpha=NamedRerouteUsage())
  n77 = Constant3Vector() {constant=<rgba 0.5 0.5 1 0>}
  n78 = DotProduct(A=@n76, B=@n77)
  n79 = ScalarParameter() {defaultValue=0.2, group=STGridFX, parameterName=AdditionalGridStrength}
  n80 = MaterialFunctionCall(Float3=MaterialFunctionCall(Normal=@n76, Flatness=OneMinus(None=@n79)) {materialFunction=&FlattenNormal}.Result) {materialFunction=&BreakOutFloat3Components}
  n81 = DotProduct(
   A=MaterialFunctionCall(X=@n80.G, Y=@n80.G, Z=@n80.G) {materialFunction=&MakeFloat3}.Result
   B=@n77
  )
  n82 = Multiply(A=Add(A=@n81, B=Time()), B=ScalarParameter() {defaultValue=0.5, group=STGridFX, parameterName=GridAnimSpeed})
  n83 = Sine(None=Frac(None=Add(A=@n78, B=@n82)))
  n84 = Saturate(None=@n83)
  n85 = Saturate(None=Add(A=@n81) {constB=-0.5})
  n86 = TextureSampleParameter2D() {
   group=06_Emissive
   parameterName=Emissive
   samplerSource=SSM_Wrap_WorldGroupSettings
   texture=&T_SpatialTravel_Backpack_E
  }
  n87 = Constant3Vector() {constant=<rgba 0 0.5962 1 1>}
  n88 = LinearInterpolate(
   A=LinearInterpolate(
    A=ScalarParameter() {defaultValue=0.5, group=STGridFX, parameterName=GridEmissiveMult_High}
    B=ScalarParameter() {defaultValue=0.1, group=STGridFX, parameterName=GridEmissiveMult_Low}
    Alpha=Power(Base=@n84, Exp=@n85)
   )
   B=NamedRerouteUsage()
   Alpha=NamedRerouteDeclaration(None=SphereMask(A=@n86.RGB, B=@n87) {attenuationRadius=0.63}) {name=Grid Line, nodeColor=<rgba 0.8281 0 1 1>}
  ) {constB=0}
  n89 = NamedRerouteUsage()
  n90 = SphereMask(A=NamedRerouteUsage(), B=Constant()) {attenuationRadius=0.1, hardnessPercent=0}
  NamedRerouteDeclaration(None=Multiply(A=Multiply(A=@n88, B=@n89), B=OneMinus(None=@n90))) {name=EmissiveOut, nodeColor=<rgba 0.3359 1 0 1>}
  NamedRerouteDeclaration(None=Power(Base=@n83, Exp=ScalarParameter() {defaultValue=1, group=STGridFX, parameterName=GridRandomMaskPower}) {constExponent=3}) {name=GridRandomMask, nodeColor=<rgba 0.4531 1 0 1>}
  n91 = TextureSample!(UVs, Tex, Apply View MipBias) {
   samplerSource=SSM_Wrap_WorldGroupSettings
   samplerType=SAMPLERTYPE_Masks
   texture=/BRCosmetics/Accessories/FORT_Backpacks/Backpack_F_MED_SpatialTravel/Textures/T_SpatialTravel_Backpack_FX3
  }
  n92 = NamedRerouteUsage()
  n93 = LinearInterpolate(
   A=Constant4Vector() {constant=<rgba 0.0501 0.0169 0.1615 1>}.RGBA
   B=Constant4Vector() {constant=<rgba 0.0333 0.7461 0.9167 1>}.RGBA
   Alpha=Saturate(None=Multiply(A=@n91.R, B=@n92))
  )
  NamedRerouteDeclaration(None=ComponentMask(None=Multiply(A=@n93, B=@n92)) {b=true, g=true, r=true}) {name=TransitionNormal, nodeColor=<rgba 1 0 0.2812 1>}
  NamedRerouteDeclaration(None=@n63.Normal) {name=NormalIn, nodeColor=<rgba 1 0.3984 0 1>}
  n94 = LinearInterpolate(
   B=Multiply(A=VectorParameter() {group=STGridFX, parameterName=Emissive_GridOutLine, defaultValue=<rgba 0 0.8 2 1>}.RGB) {constB=2}
   Alpha=NamedRerouteUsage()
  )
  n95 = Saturate(None=Subtract(A=NamedRerouteUsage(), B=NamedRerouteUsage()))
  n96 = Saturate(None=Subtract(A=NamedRerouteUsage(), B=NamedRerouteUsage()))
  n97 = VectorParameter() {parameterName=Emissive_TransitionGrids, defaultValue=<rgba 0 0.4336 0.7656 1>}
  NamedRerouteDeclaration(None=Add(A=LinearInterpolate(A=@n63.EmissiveColor, B=@n94, Alpha=@n95), B=Multiply(A=@n96, B=@n97.RGB))) {name=EmissiveIn, nodeColor=<rgba 0 1 0.625 1>}
  n98 = MaterialFunctionCall!(UVs, TextureIn, UpperLeftCorner=Constant2Vector() {g=0.6, r=0.15}, LowerRightCorner=Constant2Vector() {g=1, r=0.75}, ExportFloat4) {materialFunction=/Engine/Functions/Engine_MaterialFunctions01/Texturing/TextureCropping}
  NamedRerouteDeclaration(None=Multiply(A=@n91.B, B=@n98.Crop Mask)) {name=OutlineMask, nodeColor=<rgba 0.125 0 1 1>}
  NamedRerouteDeclaration(None=@n16) {name=DividerLastRow, nodeColor=<rgba 1 0.0703 0 1>}
  NamedRerouteDeclaration(None=@n91.R) {name=TrasitionGradient, nodeColor=<rgba 0 1 0.4844 1>}
  n99 = TextureSample!(UVs, Tex, Level, Apply View MipBias) {
   automaticViewMipBias=false
   mipValueMode=TMVM_MipLevel
   samplerSource=SSM_Wrap_WorldGroupSettings
   samplerType=SAMPLERTYPE_Masks
   texture=&T_SpatialTravel_Anim
  }
  NamedRerouteDeclaration(None=OneMinus(None=Step(Y=@n99.G) {constX=0.7})) {name=Surface_LastRow, nodeColor=<rgba 0 1 0.1328 1>}
  n100 = NamedRerouteUsage()
  n101 = SphereMask(A=Divide(A=Abs(None=@n100)) {constB=30}, B=Constant() {r=1}) {attenuationRadius=0.5, hardnessPercent=0}
  NamedRerouteDeclaration(None=LinearInterpolate(A=@n51, B=@n52, Alpha=@n101)) {name=BlankMask, nodeColor=<rgba 0 0.8828 1 1>}
  n102 = Panner(Coordinate=@n22) {speedY=-0.2}
  n103 = TextureSample!(UVs=Multiply(A=Multiply(A=@n102) {constB=2}), Tex, Apply View MipBias) {texture=/Game/Global/Textures/Default/Noise/TankMoteNoise}
  n104 = MaterialFunctionCall(Input=@n18.G, Input Low=Constant() {r=0.1}, Input High=Constant() {r=0.9}, Target Low=Constant(), Target High=Constant() {r=1}) {materialFunction=&RemapValueRange}
  n105 = LinearInterpolate(Alpha=SphereMask(A=Multiply(A=@n104.Result), B=Constant() {r=0.4}) {attenuationRadius=0.6, hardnessPercent=0})
  NamedRerouteDeclaration(None=Saturate(None=Power(Base=@n103.RGB, Exp=@n105) {constExponent=0.5})) {name=BlankMask_Bottom, nodeColor=<rgba 1 0.75 0 1>}
  NamedRerouteDeclaration(None=@n11) {name=BlankMask_Top, nodeColor=<rgba 0.4297 0 1 1>}
  NamedRerouteDeclaration(None=@n17) {name=BottomCap, nodeColor=<rgba 0 0.7891 1 1>}
  n106 = Constant2Vector() {g=0.7}
  n107 = Constant2Vector() {g=0.92, r=0.1}
  NamedRerouteDeclaration(None=Multiply(B=MaterialFunctionCall(UpperLeftCorner=@n106, LowerRightCorner=@n107) {materialFunction=&TextureCropping}.Crop Mask) {constA=1}) {name=DividerStrips, nodeColor=<rgba 0.8516 0 1 1>}
  NamedRerouteDeclaration(None=LinearInterpolate(A=@n61, Alpha=NamedRerouteUsage()) {constB=0}) {name=WipingBand, nodeColor=<rgba 0 0.7891 1 1>}
  NamedRerouteDeclaration(None=@n44) {name=LastRowDivider_Gradient, nodeColor=<rgba 1 0 0.5391 1>}
  NamedRerouteDeclaration(None=@n49) {name=GridAnimMask, nodeColor=<rgba 0.3594 1 0 1>}
  NamedRerouteDeclaration(None=@n15) {name=RotatedLocalHeight, nodeColor=<rgba 1 0.1875 0 1>}
  NamedRerouteDeclaration(None=Multiply(A=@n91.B, B=OneMinus(None=@n98.Crop Mask))) {name=BlankGirds, nodeColor=<rgba 0.4766 1 0 1>}
  n108 = NamedRerouteUsage()
  n109 = MaterialFunctionCall(Normal=@n23.RGB, Flatness=ScalarParameter() {defaultValue=0.2, group=STGridFX, parameterName=AdditionalGridFlatness}) {materialFunction=&FlattenNormal}
  NamedRerouteDeclaration(None=Add(
    A=NamedRerouteUsage()
    B=MaterialFunctionCall(BaseNormal=@n108, AdditionalNormal=@n109.Result) {materialFunction=&BlendAngleCorrectedNormals}.Result
   )) {name=NormalOut, nodeColor=<rgba 0 0.4141 1 1>}
  n110 = Add(A=Multiply(A=MaterialFunctionCall() {materialFunction=&DebugTimeSine}.Result) {constB=0.4}) {constB=0.1}
  n111 = Power(Base=SmoothStep(Min=Add(A=@n110) {constB=-0.05}, Max=Add(A=@n110) {constB=0.05}, Value=@n91.R)) {constExponent=1}
  n112 = NamedRerouteUsage()
  NamedRerouteDeclaration(None=Multiply(A=Multiply(A=@n91.G, B=@n111), B=NamedRerouteDeclaration(None=@n112) {name=TransitionArea, nodeColor=<rgba 1 0.6328 0 1>})) {name=TransitionArea_Animation, nodeColor=<rgba 1 0.6328 0 1>}
  NamedRerouteDeclaration(None=DotProduct(B=@n77)) {name=GridGrayscale, nodeColor=<rgba 0 0.3906 1 1>}
  NamedRerouteUsage()
  NamedRerouteUsage()
  NamedRerouteUsage()
  n113 = FunctionInput() {inputName=TechArtMask}
  OneMinus(None=Sign(None=ChannelMaskParameter(None=@n113) {group=STGridFX, maskChannel=Blue, parameterName=GridAnimChannel, defaultValue=<rgba 0 0 1 0>}.R))
  Saturate(None=Add(A=@n101, B=NamedRerouteUsage()))
  n114 = Sign(None=@n45.B)
  n115 = Sign(None=@n1.G)
  Sign(None=Add(A=Multiply(A=@n114, B=@n115), B=NamedRerouteDeclaration(None=@n1.B) {name=TransitionMask_B, nodeColor=<rgba 0 0.25 1 1>}))
  SphereMask(A=@n91.R, B=@n110) {attenuationRadius=0.1, hardnessPercent=0}
  Step(Y=@n110, X=@n91.R)
  Subtract(A=@n47.OpacityMask, B=NamedRerouteUsage())
  TextureSampleParameter2D() {
   group=STGridFX
   parameterName=AdditionalGrid
   samplerType=SAMPLERTYPE_Normal
   texture=&T_SpatialTravel_Backpack_N
  }
 ]}
